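; Blueprint for the Meatball (motorboat) vehicle's launcher weapon, handling its fire effects and visual setup on the host vehicle.
; On setup, a host tagged Vehicle.Meatball supplies its first "Muzzle"-tagged Cascade particle component, which is stored, and the weapon's own mesh is hidden.
; On weapon fire, it calls the parent event, then activates that stored boat muzzle component as a cooldown effect.
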(event EventOnPlayWeaponFireFX
  (params (bPersistentFire :type Boolean) (bSecondaryFire :type Boolean))
  ;[1 1 1] Play cooldown fx when fired
  (|CallParentFunction false false)
  (Components|Activation|Activate (Variables|Default|GetMuzzleFromBoat) false))

(event Setup|EventOnHostVehicleSetup
  (bind _weaponmesh (Variables|Visual|GetWeaponMesh))
  (bind _asfort_athena_vehicle (Utilities|Casting|CastToFortAthenaVehicle (Vehicle|GetHostVehicle)))
  (if (GameplayTags|HasTag (Class|FortVehicleInterface|GetVehicleTags _asfort_athena_vehicle) "(TagName=\"Vehicle.Meatball\")")
    (bind _ascascade_particle_system_component (Utilities|Casting|CastToParticleSystemComponent (Utilities|Array|Get(acopy) (Actor|GetComponentsbyTag _asfort_athena_vehicle "/Script/CascadeRuntime.ParticleSystemComponent" "Muzzle"))))
    (Variables|Default|SetMuzzleFromBoat _ascascade_particle_system_component)
    (Rendering|SetVisibility _weaponmesh false false)
    (Development|SetHiddeninGame _weaponmesh true false) ; Double verifying mesh it hidden
  ))

(fn ConstructionScript
  (|CallParentFunction))
