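; Motorboat missile projectile that homes in DrunkArtillery mode, spins in flight, and explodes on stopping or when its six-second fuse expires.
; Touch hits either stop it or, on NoStick components, double its velocity.
; Explosions can launch Meatball-tagged boats away with a minimum upward impulse, and play water or land effects, surface-specific FX, sound and camera shake behind cosmetic checks.
; Construction loads turn-rate and launch tuning from hotfix scalable-float rows.
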(event EventBeginPlay
  ;[1 1 1] Begin Play
  (Projectile|SetHomingData 0 0 (Utilities|Struct|MakeProjectileHomingData "DrunkArtillery" (Variables|HomingTurnRate|GetHomingMinTurnSpeed) (Variables|HomingTurnRate|GetHomingMaxTurnSpeed) (Variables|HomingTurnRate|GetTurnRateRampTime) 1e+01) :!class /Script/FortniteGame.FortProjectileMovementInterface)
  (Utilities|Time|SetTimerbyFunctionName self "FuseTimerMax" 6.0)
  (Networking|CanExecuteCosmeticEvents
    (:True
      (Components|Activation|Activate (Variables|Default|Get%20RocketTrailPS) false)
      (Rendering|SetRenderCustomDepth (Variables|Default|Get%20RocketTrailPS) true))))

(event EventTick
  (params (DeltaSeconds :type Float%20%28single-precision%29))
  ;[1 1 1] On Tick, spin the rocket and particle trail
  (Variables|Default|SetTick%20Delta DeltaSeconds)
  (Networking|CanExecuteCosmeticEvents
    (:True
      (if (Components|Activation|IsActive (Variables|Default|Get%20RocketTrailPS))
        (Variables|Effects|SetRocketSpin (Math|Rotator|MakeRotator (* (Variables|Default|GetTick%20Delta) (Variables|Effects|GetRocketSpinRate))))
        (Transformation|AddLocalRotation (Variables|Default|GetMissile) (Variables|Effects|GetRocketSpin) false false)))))

(event EventOnTouched
  (params (OtherActor :type Actor%20Object%20Reference) (OtherComp :type Primitive%20Component%20Object%20Reference) (HitResult :type Hit%20Result%20Structure%20%28by%20ref%29) (bIsOverlap :type Boolean))
  (bind (_bblockinghit _binitialoverlap _time _distance _location _impactpoint _normal _impactnormal _physmat) (Collision|BreakHitResult HitResult))
  (bind _surfacetype (Variables|/Script/PhysicsCore.PhysicalMaterial|GetSurfaceType _physmat))
  (bind (_bblockinghit_2 _binitialoverlap_2 _time_2 _distance_2 _location_2 _impactpoint_2 _normal_2 _impactnormal_2 _physmat_2 _hitactor _hitcomponent) (Collision|BreakHitResult HitResult))
  ;[1 1 1] Bounce off "NoBounce"
  (Utilities|IsValid _physmat
    (:Is%20Valid
      (if (or (not bIsOverlap) (Utilities|Enum|Equal(Enum) _surfacetype "SurfaceType8"))
        (if (Components|ComponentHasTag _hitcomponent "NoStick")
          (Variables|/Script/Engine.MovementComponent|SetVelocity (* (Variables|/Script/Engine.MovementComponent|GetVelocity (Variables|Movement|GetProjectileMovementComponent)) 2.0) (Variables|Movement|GetProjectileMovementComponent))
          (else
            (Variables|Projectile|SetbStopSimulatingOnHit true)
            (Variables|Default|SetSurfaceType _surfacetype)))))))

(event EventOnStop
  (params (Hit :type Hit%20Result%20Structure%20%28by%20ref%29))
  (bind (_bblockinghit _binitialoverlap _time _distance _location _impactpoint _normal _impactnormal _physmat) (Collision|BreakHitResult Hit))
  ;[1 1 1] Impact and Explosion Stuff
  (Variables|Effects|SetDecalLocation _impactpoint)
  (Variables|Effects|SetStopLocation (+ _impactpoint (* _impactnormal (Variables|Effects|GetExplosion%20Impact%20Offset))))
  (Variables|Effects|SetStopRotZ (Math|Rotator|MakeRotfromZ _impactnormal))
  (Utilities|IsValid _physmat
    (:Is%20Valid
      (Variables|Default|SetSurfaceType (Variables|/Script/PhysicsCore.PhysicalMaterial|GetSurfaceType _physmat)))
    (:Is%20Not%20Valid
      _))
  (if (not (Delivery|IsBeingKilled))
    (Delivery|DoExplosion)
    (Fortnite|ShowSoundIndicatorLocal :Location (Transformation|GetWorldLocation (Variables|Default|GetMissile)) :MaxAudibleDistance 5e+03 :IndicatorType "Gunshot" :Instigator self :Tint "(R=0.000000,G=0.000000,B=0.000000,A=0.000000)" :AttenuationMaxRangeMultiplier 1.0 :AttenuationIntensityMultiplier 1.0)))

(event Custom|FuseTimerMax
  (bind _missile (Variables|Default|GetMissile))
  (bind _returnvalue (Transformation|GetWorldLocation _missile))
  ;[1 1 1] Impact and Explosion Stuff
  (Variables|Effects|SetTimerMaxReached? true)
  (Variables|Effects|SetStopLocation _returnvalue)
  (if (not (Delivery|IsBeingKilled))
    (Delivery|DoExplosion)
    (Fortnite|ShowSoundIndicatorLocal :Location _returnvalue :MaxAudibleDistance 5e+03 :IndicatorType "Gunshot" :Instigator self :Tint "(R=0.000000,G=0.000000,B=0.000000,A=0.000000)" :AttenuationMaxRangeMultiplier 1.0 :AttenuationIntensityMultiplier 1.0)))

(event EventOnExploded
  (params (HitActors :type Array%20of%20Actor%20Object%20References) (HitResults :type Array%20of%20Hit%20Result%20Structures))
  (bind _surfacetype (Variables|Default|GetSurfaceType))
  (bind _returnvalue (Math|Conversions|ToInteger(Byte) _surfacetype))
  (bind _stoplocation (Variables|Effects|GetStopLocation))
  (bind _stoprotz (Variables|Effects|GetStopRotZ))
  ;[1 1 1] Impact and Explosion Stuff
  (Utilities|FlowControl|Sequence
    (:then_0
      ;;[1 1 1] If we hit a boat --> Launch it
      (if (> (Variables|Default|GetShouldImpulseBoats) 0.0)
        (for _array_element HitActors
          (bind _asfort_athena_vehicle (Utilities|Casting|CastToFortAthenaVehicle _array_element))
          (if (GameplayTags|HasTag (Class|FortVehicleInterface|GetVehicleTags _asfort_athena_vehicle) "(TagName=\"Vehicle.Meatball\")")
            (Audio|BroadcastAbilityImpactatLocation 0 _asfort_athena_vehicle (CallFunction|CalcLaunchVelocity _asfort_athena_vehicle) 777.0)))))
    (:then_1
      (Networking|CanExecuteCosmeticEvents
        (:True
          (Components|Activation|SetActive (Variables|Default|Get%20RocketTrailPS) false false)
          ;[1 1 1] SingleLineCheckForWorldObjectsSurfaceType
          (if (or (Utilities|Enum|Equal(Enum) (Variables|Default|GetSurfaceType) "SurfaceType8") (Water|IsinWater (Variables|Water|GetWaterInteractionComponent)))
            (bind _returnvalue_2 (Effects|Components|ParticleSystem|SpawnFXAtLocation (Variables|Default|GetFX_Water_Explosion_Asset) _stoplocation _stoprotz "1.000000,1.000000,1.000000" true "AutoRelease"))
            (Rendering|SetRenderCustomDepth _returnvalue_2 true)
            (Audio|PlaySoundatLocation (Variables|Sounds|GetRocket_Projectile_Explosion_Water_Sound) (Variables|Effects|GetStopLocation))
            (Audio|PlaySoundatLocation "/Game/Athena/Sounds/Weapons/Explosions/Expl_Debris_Water_Cue.Expl_Debris_Water_Cue" (Variables|Effects|GetStopLocation))
            (else
              (bind _returnvalue_3 (Effects|Components|ParticleSystem|SpawnFXAtLocation (Variables|Default|GetFX_Explosion_Asset) _stoplocation _stoprotz "1.000000,1.000000,1.000000" true "AutoRelease"))
              (Rendering|SetRenderCustomDepth _returnvalue_3 true)
              ;;[1 1 1] PhysSurface Interaction for explosion Niagara FX
              (bind _checkedphysicalsurfaceid (CallFunction|PhysicalSurfaceCheck _returnvalue))
              (if _checkedphysicalsurfaceid
                (bind _returnvalue_4 (Effects|Components|ParticleSystem|SpawnFXAtLocation (Variables|Default|GetFX_Physical%20Surface) _stoplocation))
                (Effects|Components|ParticleSystem|SetIntParameter _returnvalue_4 "PhysicalSurface" _returnvalue)
                (Effects|Components|ParticleSystem|SetFloatParameter _returnvalue_4 "Global Scale" 0.9)
                (else
                  _))
              (Audio|PlaySoundatLocation (Variables|Sounds|GetRocket_Projectile_Explosion_Sound) (Variables|Effects|GetStopLocation))))
          (Rendering|SetVisibility (Variables|Default|GetMissile) false false)
          ;[1 1 1] Shake that camera
          (GameplayCue|ExecuteGameplayCue%28Burst%29Local self (Variables|Effects|GetFeedbackCue) (Ability|GameplayCue|MakeGameplayCueParameters 0.0)) ; Cam shake + controller vibration
          (Audio|Components|Audio|FadeOut (Variables|Default|GetRocket_AudioComponent) 1 0.0 "Linear" :!class /Script/Engine.AudioComponent)
          (Utilities|AllowJanus
            (:Allow
              (bind _returnvalue_5 (Development|GetConsoleVariableIntValue "r.PhysicsField.EnableField"))
              (if (== _returnvalue_5 1)
                (Game|SpawnActorfromClass "/Game/Effects/Niagara/Blueprints/PhsyicsFields/FN_RadialForce.FN_RadialForce_C" (Transformation|GetWorldTransform (Variables|Collision|GetCapsuleComponent)) "Undefined" "MultiplyWithRoot" 0 0.2 0.75 1.5e+03 2.0 0.0 0)))))
        (:False
          (CallFunction|HideonExplode))))))

(event EventDestroyed
  ;[1 1 1] Impact and Explosion Stuff
  (Networking|CanExecuteCosmeticEvents
    (:True
      (Components|Activation|SetActive (Variables|Default|Get%20RocketTrailPS) false false))))

(event Custom|HideOnExplode
  ;[1 1 1] Impact and Explosion Stuff
  (Development|SetHiddeninGame (Variables|Default|GetMissile) true false))

(fn CalcLaunchVelocity
  (params (Actor :type Actor%20Object%20Reference))
  (return (Math|Vector|MakeVector (.x (* (Math|Vector|Normalize (- (Transformation|GetActorLocation Actor) (Variables|Effects|GetStopLocation))) (Variables|Default|GetBoatLaunchMultiplier))) (.y (* (Math|Vector|Normalize (- (Transformation|GetActorLocation Actor) (Variables|Effects|GetStopLocation))) (Variables|Default|GetBoatLaunchMultiplier))) (Math|Float|Max(Float) (+ (.z (* (Math|Vector|Normalize (- (Transformation|GetActorLocation Actor) (Variables|Effects|GetStopLocation))) (Variables|Default|GetBoatLaunchMultiplier))) (Variables|Default|GetZ_Additive)) (Variables|Default|GetMinZ_Impulse)))))

(fn PhysicalSurfaceCheck
  (params (Physical%20Surface%20ID :type Integer))
  (return (Math|Boolean|ORBoolean (== Physical%20Surface%20ID 30) (== Physical%20Surface%20ID 26) (== Physical%20Surface%20ID 23) (== Physical%20Surface%20ID 24))))

(fn ConstructionScript
  (Effects|Components|ParticleSystem|SetTemplate (Variables|Default|GetRibbon_Trail_PSC) (Variables|Effects|GetRibbon_Trail_PS))
  (Variables|Default|SetFX_Explosion_Asset (select (Effects|IsNiagaraFXTypeEnabled "Projectile") "/Game/Effects/Explosion_Update/NS_GrenadeLauncher_Explosion_Asteria.NS_GrenadeLauncher_Explosion_Asteria" "/Game/Effects/Fort_Effects/Gameplay/Ostrich/P_Ostrich_Rocket_Explosion.P_Ostrich_Rocket_Explosion"))
  (Effects|Components|ParticleSystem|SetTemplate (Variables|Default|GetExplosion_Flesh_Damage_PSC) (Variables|Effects|GetExplosion_Flesh_Damage_PS))
  (Variables|Default|SetFX_Water_Explosion_Asset (select (Effects|IsNiagaraFXTypeEnabled "Projectile") "/Game/Effects/Explosion_Update/NS_GrenadeLauncher_Explosion_Asteria_Water.NS_GrenadeLauncher_Explosion_Asteria_Water" 0))
  (Effects|Components|ParticleSystem|SetTemplate (Variables|Default|GetWater_Explosion_PSC) (Variables|Effects|GetWater_Explosion_PS))
  (Variables|HomingTurnRate|SetHomingMinTurnSpeed (ScalableFloat|GetValueatLevel (Variables|HotfixRows|GetRow_MinTurnSpeed)))
  (Variables|HomingTurnRate|SetHomingMaxTurnSpeed (ScalableFloat|GetValueatLevel (Variables|HotfixRows|GetRow_MaxTurnSpeed)))
  (Variables|HomingTurnRate|SetTurnRateRampTime (ScalableFloat|GetValueatLevel (Variables|HotfixRows|GetRow_TurnRateRampTime)))
  (Variables|Default|SetBoatLaunchMultiplier (ScalableFloat|GetValueatLevel (Variables|HotfixRows|GetRow_BoatLaunchMultiplier)))
  (Variables|Default|SetMinZ_Impulse (ScalableFloat|GetValueatLevel (Variables|HotfixRows|GetRow_MinZImpulse)))
  (Variables|Default|SetZ_Additive (ScalableFloat|GetValueatLevel (Variables|HotfixRows|GetRow_ZImpulseAdditive)))
  (Variables|Default|SetShouldImpulseBoats (ScalableFloat|GetValueatLevel (Variables|HotfixRows|GetRow_ShouldImpulseBoats))))
